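# Niagara effect system for the V3 Ebony VinderTech glider, with red-to-purple color variants driven by User.Variant.
# Emitters pin electric arcs, sparks, flares, glows with lights and mesh trails to skeletal mesh sockets, alongside torus-spawned speed lines.
# System timers, timelines and a smoothed moving average drive intensity, size and flashes.
# Front-end modifiers give several emitters separate lobby and gameplay values.
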NiagaraSystem /BRCosmetics/Gadgets/Assets/VinderTech_GliderChute/Glider_Ebony/FX/V3/NS_Glider_Ebony_V3
dependencies {
 usedDataInterfaces [
  &NiagaraDataInterfaceCurve
  &NiagaraDataInterfaceArrayFloat
  &NiagaraDataInterfaceArrayFloat3
  &NiagaraDataInterfaceVector2DCurve
  &NiagaraDataInterfaceColorCurve
 ]
 usedDynamicInputs [&SetBoolByFloatComparison, &Abs_Float]
 usedModules [
  &self:SystemSpawnScript.SystemScriptSource.SystemScriptGraph.NiagaraNodeAssignment_0.SetVariables_#1
  &SystemState
  &InterpolateOverTime
  &self:SystemSpawnScript.SystemScriptSource.SystemScriptGraph.NiagaraNodeAssignment_1.SetVariables_#2
  &FN_Timer
  &self:SystemSpawnScript.SystemScriptSource.SystemScriptGraph.NiagaraNodeAssignment_2.SetVariables_#3
  &Timeline
  &self:SystemSpawnScript.SystemScriptSource.SystemScriptGraph.NiagaraNodeAssignment_3.SetVariables_#4
 ]
}
emitters {
 ElectricArcsSmall {
  emitterUpdateScript [
   &EmitterState(
    Life Cycle Mode=::NewEnumerator1
    Loop Delay=&RandomRangeFloat(Minimum=0.001, Maximum=0.25, Evaluation Type=::NewEnumerator1)
    Loop Duration=&RandomRangeFloat(Minimum=0.25, Maximum=1, Evaluation Type=::NewEnumerator1)
    Recalculate Duration Each Loop=true
    UseLoopDelay=true
   )
   &SpawnRate(
    Spawn Probability=0.8
    SpawnRate=&Lerp_Float(
     A=&FN_FrontEndFloatModifier(Gameplay=0, Front End=50)
     B=50
     Alpha=$System.InterpolateOverTime.InterpolateOverTime_Float.Moving Average
    )
    Use Spawn Probability=true
   )
  ]
  particleSpawnScript [
   &InitializeParticle(
    Color=&LerpLinearColors(StartColor=<rgba 40 0 0 1>, EndColor=<rgba 34.6667 0 40 1>, LerpFactor=$User.Variant)
    Lifetime=&RandomRangeFloat(Minimum=0.15, Maximum=0.25)
    Sprite Rotation Angle=&RandomRangeFloat(Minimum=0, Maximum=1)
    Sprite Rotation Mode=::NewEnumerator1
    Sprite Size Max=<25 35>
    Sprite Size Min=<15 30>
    Sprite Size Mode=::NewEnumerator2
    UsePositionOffset=false
    Write Material Random=true
   ) {module=InitializeParticle001}
   &Color(Color=&MakeLinearColorFromVectorAndFloat(
     "Vector (RGB)"=&RandomRangeVector(Minimum=<125 125 125>, Maximum=<200 200 200>)
     "Float (Alpha)"=&ClampFloat(Float=1)
    )) {enabled=false}
   &DynamicMaterialParameters(
    Index 0 Param 1=&RandomRangeFloat(Minimum=2.5, Maximum=3.5)
    Index 0 Param 3=&RandomRangeFloat(Minimum=0.2, Maximum=0.7)
    Module.Param0WriteEnabled=true
    Module.Param1WriteEnabled=false
    Module.Param2WriteEnabled=true
    Module.Param3WriteEnabled=true
   )
   &ShapeLocation(MaskSpawnGroup=false, Override Seed=false, Sphere Radius=15, UseOwnerScale=true)
   &AddVelocity(Velocity=&FN_FrontEndVector3Modifier(
     Gameplay=&RandomRangeVector(Minimum=<0 100 20>, Maximum=<1 100 30>)
     Front End=&RandomRangeVector(Minimum=<0 -100 20>, Maximum=<1 -100 30>)
    )) {enabled=false}
   &AddVelocity(
    Velocity=&FN_FrontEndVector3Modifier(Gameplay=<0 400 -120>, Front End=<0 800 -130>)
    Velocity Speed Scale=0.3
   ) {enabled=false, module=AddVelocity001}
  ]
  particleUpdateScript [
   &ParticleState()
   &SkeletalMeshLocation(
    MaskSpawnGroup=false
    Mesh Sampling Type=::NewEnumerator5
    Position Sampling=::NewEnumerator1
    Sampled Position Offset=&RandomRangeVector(Minimum=<-60 0 -5>, Maximum=<60 1 -5>)
    Skeletal Mesh=$User.SkeletalMesh
    Socket Sampling Mode=::NewEnumerator5
   ) {module=SkeletalMeshLocation001}
   &ScaleSpriteSize(Uniform Curve Scale=4)
   &ScaleColor(
    Scale Alpha=&FloatFromCurve(FloatCurve=DataInterface() {propertyValues {Curve {keys [<at 0.2 0>,<at 0.25 1>]}}})
    ScaleA=true
    ScaleRGB=true
   )
   &DynamicMaterialParameters(
    Index 0 Param 2=&FloatFromCurve(FloatCurve=DataInterface() {propertyValues {Curve {keys [<at 0 1>,<at 1 -1>]}}})
    Module.Param0WriteEnabled=false
    Module.Param1WriteEnabled=true
    Module.Param2WriteEnabled=false
    Module.Param3WriteEnabled=false
   ) {module=DynamicMaterialParameters001}
   &SolveForcesAndVelocity(Clamp Velocity=false, Limit Acceleration=false)
   &CameraOffset(Camera Offset Amount=12.5)
  ]
  rendererClasses [&NiagaraSpriteRendererProperties]
  renderers [{rendererClass=&NiagaraSpriteRendererProperties}]
 }
 FN_Debug {
  bEnabled=false
  emitterUpdateScript [
   &EmitterState()
   &SpawnBurst_Instantaneous(Spawn Count=1, Use Spawn Probability=false)
  ]
  particleSpawnScript [/Niagara/Modules/Spawn/Initialization/InitializeParticle(
    Lifetime=5
    Module.Write Color=true
    Module.Write Lifetime=true
    Module.Write Mass=true
    Module.Write Material Random=true
    Module.Write Position=true
    Module.Write Scale=false
    Module.Write SpriteRotation=false
    Module.Write SpriteSize=true
    Position=&AddVectorToPosition(Position=&SimulationPosition(), Vector=<200 0 20>)
    Sprite Size=<100 100>
   )]
  particleUpdateScript [
   &ParticleState()
   /SetVariables(Particles.Debug:NiagaraFloat=$System.Timeline.Time) {isSetParameters=true}
   &DynamicMaterialParameters(
    Index 0 Param 1=$Particles.Debug
    Module.Param0WriteEnabled=true
    Module.Param1WriteEnabled=true
    Module.Param2WriteEnabled=true
    Module.Param3WriteEnabled=true
   )
  ]
  rendererClasses [&NiagaraSpriteRendererProperties]
  renderers [{rendererClass=&NiagaraSpriteRendererProperties}]
 }
 FakeLight {
  emitterSpawnScript [&FN_PreserveSkydiveUserVariables() {module=PreserveUserVariables}]
  emitterUpdateScript [
   &EmitterState(Inactive Response=::NewEnumerator1, Life Cycle Mode=::NewEnumerator1, Loop Duration=0, UseLoopDelay=false)
   &SpawnBurst_Instantaneous(Spawn Count=2, Spawn Probability=HlslExpression() {hlslExpression="Emitter.LoopCount == 0"}, Use Spawn Probability=true) {module=SpawnBurst_Instantaneous001}
  ]
  particleSpawnScript [
   &FN_SetSpawnGroup_v2(Number of Spawn Groups=2) {enabled=false}
   &FN_SetSpawnGroup(Number of Spawn Groups=2)
   /Niagara/Modules/Spawn/Initialization/V2/InitializeParticle(
    Color=&LerpLinearColors(
     StartColor=&LerpLinearColors(
      StartColor=<rgba 4 0 0.0667 1>
      EndColor=<rgba 20 0 0.3333 1>
      LerpFactor=&ClampFloat(Float=$System.InterpolateOverTime.InterpolateOverTime_Float.Moving Average)
     )
     EndColor=&LerpLinearColors(
      StartColor=<rgba 2.9333 0 4 1>
      EndColor=<rgba 7 0 20 1>
      LerpFactor=&ClampFloat(Float=$System.InterpolateOverTime.InterpolateOverTime_Float.Moving Average)
     )
     LerpFactor=$User.Variant
    )
    Lifetime=0
    Sprite Size Mode=::NewEnumerator3
    Uniform Sprite Size=50
    UsePositionOffset=false
    Write Material Random=true
   ) {module=InitializeParticle001}
   &CameraOffset()
   &DynamicMaterialParameters(
    Index 0 Param 2=&FN_FrontEndFloatModifier(Gameplay=0, Front End=1)
    Module.Param0WriteEnabled=true
    Module.Param1WriteEnabled=true
    Module.Param2WriteEnabled=false
    Module.Param3WriteEnabled=false
   )
  ]
  particleUpdateScript [
   &ParticleState(Kill Particles When Lifetime Has Elapsed=false)
   &Color(Color=&LerpLinearColors(
     StartColor=&LerpLinearColors(
      StartColor=<rgba 1 0 0.0667 1>
      EndColor=<rgba 20 0 0.3333 1>
      LerpFactor=&ClampFloat(Float=$System.InterpolateOverTime.InterpolateOverTime_Float.Moving Average)
     )
     EndColor=&LerpLinearColors(
      StartColor=<rgba 2.9333 0 4 1>
      EndColor=<rgba 7 0 20 1>
      LerpFactor=&ClampFloat(Float=$System.InterpolateOverTime.InterpolateOverTime_Float.Moving Average)
     )
     LerpFactor=$User.Variant
    ))
   &ScaleSpriteSize(Uniform Curve Sprite Scale=DataInterface() {propertyValues {Curve {keys [<at 0.2 1>,<at 0.25 2>,<at 0.4 0>]}}}) {enabled=false}
   &SampleSkeletalMeshSkeleton(
    Mesh:NiagaraDataInterfaceSkeletalMesh=$User.SkeletalMesh
    Sampled Rotation Coordinate Space:ENiagaraCoordinateSpace=::Local
    Specific Bone Index:NiagaraInt32=$Particles.SpawnGroup
   ) {enabled=false}
   &SkeletalMeshLocation(
    Apply Sampled Orientation to Mesh Particles=true
    Bone / Socket Index=$Particles.SpawnGroup
    Bone Sampling Mode=::NewEnumerator4
    Inherit Velocity=false
    Interpolate Mesh Sampling=false
    MaskSpawnGroup=false
    Sampled Position Offset=&MaskVectorBySpawnGroup(Masked Vector=<10 0 0>, Passthrough Vector=<-10 0 0>)
    Sampled Position Offset Coordinate Space:ENiagaraMeshTransforms=::NewEnumerator3
    Skeletal Mesh=$User.SkeletalMesh
   ) {enabled=false}
   &SkeletalMeshLocation(
    MaskSpawnGroup=false
    Mesh Sampling Type=::NewEnumerator5
    Sampled Position Offset=&SelectVectorFromArray(Direct Array Index=$Particles.SpawnGroup, Vector Selection Array=DataInterface() {propertyValues {FloatData [<0 15 20>, <0 15 -20>]}})
    Skeletal Mesh=$User.SkeletalMesh
    Socket Sampling Mode=::NewEnumerator5
   ) {module=SkeletalMeshLocation001}
   /SetVariables(Particles.Color=&MultiplyLinearColor(
     Linear Color=$StackContext.Color
     Scale Factor=&MakeVector4_FromFloat(Float=&FloatFromCurve(
       FloatCurve=DataInterface() {propertyValues {Curve {keys [<at 0.2 1>,<at 0.25 2>,<at 0.4 1>]}}}
       CurveIndex=$System.Timeline001.Time
       Scale Curve=0.5
      ))
    )) {enabled=false, isSetParameters=true}
   &ScaleColor(
    Color Value To Scale=$StackContext.Color
    Scale RGB=&VectorFromFloat(Value=&ClampFloat(
      Float=&FloatFromCurve(FloatCurve=DataInterface() {propertyValues {Curve {keys [<at 0.2 1>,<at 0.25 3>,<at 0.4 1>]}}}, CurveIndex=$System.Timeline001.Time)
      Max=5
     ))
    ScaleA=true
    ScaleRGB=true
   )
  ]
  rendererClasses [&NiagaraSpriteRendererProperties]
  renderers [
   {rendererClass=&NiagaraSpriteRendererProperties}
   {rendererClass=&NiagaraSpriteRendererProperties, rendererIndex=1}
  ]
 }
 Flare {
  emitterUpdateScript [
   &EmitterState(
    Life Cycle Mode=::NewEnumerator1
    Loop Behavior=::NewEnumerator1
    Loop Delay=$User.EngineDelay
    Loop Duration Mode=::NewEnumerator1
    UseLoopDelay=true
   )
   &SpawnBurst_Instantaneous(Spawn Probability=HlslExpression() {hlslExpression="Emitter.LoopCount == 0"}, Use Spawn Probability=true, UseLoopCountLimit=false)
  ]
  particleSpawnScript [
   &FN_SetSpawnGroup(Number of Spawn Groups=1)
   &InitializeParticle(
    Color=&LerpLinearColors(StartColor=<rgba 5 0 0 1>, EndColor=<rgba 3.75 0 5 1>, LerpFactor=$User.Variant)
    Lifetime=0.25
    Position Offset=<0 0 50>
    Sprite Size=&Multiply_Vector2D_ByFloat(
     Float=&SelectFloatFromArray(Direct Array Index=$Particles.SpawnGroup, Float Selection Array=DataInterface() {propertyValues {FloatData [1.25,1.25,3]}})
     Vector2D=<80 50>
    )
    Sprite Size Mode=::NewEnumerator4
    UsePositionOffset=false
    Write Material Random=true
   )
  ]
  particleUpdateScript [
   &ParticleState(Kill Particles When Lifetime Has Elapsed=false, Let Infinitely Lived Particles Die When Emitter Deactivates=true)
   &SkeletalMeshLocation(
    MaskSpawnGroup=false
    Mesh Sampling Type=::NewEnumerator5
    Sampled Position Offset=<0 10 -5>
    Skeletal Mesh=$User.SkeletalMesh
    Socket Sampling Mode=::NewEnumerator5
   ) {module=SkeletalMeshLocation001}
   &ScaleSpriteSize(
    Scale Sprite Size Mode=::NewEnumerator0
    Uniform Scale Factor=&Multiply_Float(
     A=&FloatFromCurve(FloatCurve=DataInterface() {propertyValues {Curve {keys [<at 0 0>,<at 1 1>]}}}, Scale Curve=1.2)
     B=&Lerp_Float(A=0.6, B=0.8, Alpha=$System.InterpolateOverTime.InterpolateOverTime_Float.Moving Average)
    )
   )
   &ScaleColor(
    Scale Alpha=&Multiply_Float(
     A=&MakeFloatFromVector4(
      Vector 4=&MakeVector4_FromFloat(Float=&Lerp_Float(A=0, B=3, Alpha=$System.InterpolateOverTime.InterpolateOverTime_Float.Moving Average))
      Channel=::NewEnumerator0
     )
     B=&FNTimeOfDayScalar(Night Scale=1, Morning Scale=1, Evening Scale=1)
    )
    ScaleA=true
    ScaleRGB=true
   )
   &CameraOffset(Camera Offset Amount=20)
   &DynamicMaterialParameters(
    Index 0 Param 1=0.4
    Index 0 Param 2=0.5
    Module.Param0WriteEnabled=true
    Module.Param1WriteEnabled=true
    Module.Param2WriteEnabled=true
    Module.Param3WriteEnabled=true
   )
   &FN_Light_v2(
    Light Exponent=2
    Light Radius:NiagaraFloat=&SelectFloatFromArray(Direct Array Index=$Particles.SpawnGroup, Float Selection Array=DataInterface() {propertyValues {FloatData [40,40,80]}})
    LightEnabled:NiagaraBool=$User.bIsFullyDeployed
    Position Offset:Vector3f=<0 -25 0>
    Scale Color=10
   )
   /SetVariables(Particles.Light Color:LinearColor=&MakeLinearColor(
     R:NiagaraFloat=50
     G:NiagaraFloat=0
     B:NiagaraFloat=0
     A:NiagaraFloat=&MakeFloatFromVector4(
      Vector 4=&MakeVector4_FromFloat(Float=&Add_Float(
        A=&Multiply_Float(
         A=&Sine(Normalized Angle=$Emitter.Age, Period=0.2, Scale=0.3, Bias=1)
         B=&Lerp_Float(
          A=0
          B=0.2
          Alpha=&ClampFloat(Float=$System.InterpolateOverTime.InterpolateOverTime_Float.Moving Average, Max=0.5)
         )
        )
        B=&FloatFromCurve(
         FloatCurve=DataInterface() {propertyValues {Curve {keys [<at 0.2 0>,<at 0.25 20>,<at 0.4 0>]}}}
         CurveIndex=$System.Timeline001.Time
         Scale Curve=4
        )
       ))
      Channel=::NewEnumerator0
     )
    )) {isSetParameters=true}
  ]
  rendererClasses [&NiagaraSpriteRendererProperties, &NiagaraLightRendererProperties]
  renderers [
   {rendererClass=&NiagaraSpriteRendererProperties}
   {rendererClass=&NiagaraLightRendererProperties, rendererIndex=1}
  ]
 }
 GLOW_REACTORS001 {
  emitterUpdateScript [
   &EmitterState(
    Inactive Response=::NewEnumerator1
    Life Cycle Mode=::NewEnumerator1
    Loop Duration=0
    Recalculate Duration Each Loop=true
    UseLoopDelay=false
   )
   &SpawnBurst_Instantaneous(Spawn Probability=HlslExpression() {hlslExpression="Emitter.LoopCount == 0"}, Use Spawn Probability=true, UseLoopCountLimit=false)
  ]
  particleSpawnScript [
   &InitializeParticle(
    Color=<rgba 1 0.01 0.01 0.15>
    Lifetime Max=0.7
    Lifetime Min=0.5
    Lifetime Mode=::NewEnumerator1
    Sprite Size Mode=::NewEnumerator3
    Uniform Sprite Size=80
    UsePositionOffset=false
    Write Material Random=true
   )
   &FN_SetSpawnGroup(Number of Spawn Groups=3)
   &CameraOffset(Camera Offset Amount=5)
   /SetVariables(Particles.FacingCamera:Vector3f=<0 1 0>) {isSetParameters=true}
   /SetVariables(Particles.CamreaOffsetLow=4) {isSetParameters=true}
  ]
  particleUpdateScript [
   &SolveForcesAndVelocity(Clamp Velocity=false, Limit Acceleration=false)
   &ParticleState(
    Kill Particles When Lifetime Has Elapsed=false
    Let Infinitely Lived Particles Die When Emitter Deactivates=true
    Loop Particles Lifetime=true
   )
   &SkeletalMeshLocation(
    MaskSpawnGroup=false
    Mesh Sampling Type=::NewEnumerator5
    Sampled Position Offset=<0 0 5>
    Skeletal Mesh=$User.SkeletalMesh
    Socket Sampling Mode=::NewEnumerator5
   )
   &DynamicMaterialParameters(
    Index 0 Param 1=50
    Module.Param0WriteEnabled=true
    Module.Param1WriteEnabled=true
    Module.Param2WriteEnabled=true
    Module.Param3WriteEnabled=true
   )
   &ScaleColor(
    Scale Alpha=&Multiply_Float(
     A=&Sine(
      Normalized Angle=&Subtract_Float(A=$Emitter.Age, B=&ReturnNormalizedExecIndex())
      Period=0.5
      Scale=0.2
      Bias=1
     )
     B=&Lerp_Float(A=0.2, B=0.5, Alpha=$System.InterpolateOverTime.InterpolateOverTime_Float.Moving Average)
    )
    ScaleA=true
    ScaleRGB=true
   )
   &FN_Light_v2(
    Light Exponent=4
    Light Radius:NiagaraFloat=&Multiply_Float(
     A=0.9
     B=&SelectFloatFromArray(Direct Array Index=$Particles.SpawnGroup, Float Selection Array=DataInterface() {propertyValues {FloatData [40,40,50]}})
    )
    LightEnabled:NiagaraBool=$User.bIsFullyDeployed
    Scale Color=1
   )
   /SetVariables(Particles.Light Color:LinearColor=&MultiplyLinearColor(
     Linear Color=<rgba 20 0 0 1>
     Scale Factor=&MakeVector4_FromFloat(Float=&Multiply_Float(
       A=&Multiply_Float(
        A=&Sine(Normalized Angle=$Emitter.Age, Period=0.4, Scale=0.2, Bias=1)
        B=&Sine(Normalized Angle=$Emitter.Age, Period=0.2, Scale=0.1, Bias=1)
       )
       B=&Lerp_Float(A=0, B=1, Alpha=$System.InterpolateOverTime.InterpolateOverTime_Float.Moving Average)
      ))
    )) {isSetParameters=true}
  ]
  rendererClasses [&NiagaraSpriteRendererProperties, &NiagaraLightRendererProperties]
  renderers [
   {rendererClass=&NiagaraSpriteRendererProperties}
   {rendererClass=&NiagaraLightRendererProperties, rendererIndex=1}
  ]
 }
 Sparks {
  emitterUpdateScript [
   &EmitterState(
    Delay First Loop Only=true
    Inactive Response=::NewEnumerator1
    Life Cycle Mode=::NewEnumerator1
    Loop Delay=$System.Delay
    Loop Duration=1
    UseLoopDelay=true
   )
   &SpawnRate(
    Spawn Probability=0.8
    SpawnRate=&FN_FrontEndFloatModifier(
     Gameplay=&Lerp_Float(
      A=&FloatFromCurve(
       FloatCurve=DataInterface() {propertyValues {Curve {keys [<at 0.1 0>,<at 0.25 20>,<at 0.5 2>]}}}
       CurveIndex=$System.Timeline001.Time
       Scale Curve=&Multiply_Float(A=10, B=&ClampFloat(Float=$System.InterpolateOverTime.InterpolateOverTime_Float.Moving Average))
      )
      B=20
      Alpha=$User.Debug
     )
     Front End=15
    )
    Use Spawn Probability=true
   )
  ]
  particleSpawnScript [
   &FN_SetSpawnGroup()
   &InitializeParticle(
    Color=&LerpLinearColors(StartColor=<rgba 10 0 0.1667 20>, EndColor=<rgba 10 0 7 20>, LerpFactor=$User.Variant)
    Lifetime=&RandomRangeFloat(Minimum=0.4, Maximum=0.8)
    Sprite Size Max=<1.2 7>
    Sprite Size Min=<0.7 4>
    Sprite Size Mode=::NewEnumerator2
    UsePositionOffset=false
    Write Material Random=true
   ) {module=InitializeParticle001}
   &SphereLocation(
    MaskSpawnGroup=false
    Override Seed=false
    Sphere Radius=5
    Surface Only=true
    UseNonUniformScale=false
    UseOffset=false
    UseSphereOrientation=false
   )
   &AddVelocity(
    Scale Added Velocity=&RandomRangeVector(Minimum=<2.5 2.5 2.5>, Maximum=<4 4 4>)
    Velocity=&FN_FrontEndVector3Modifier(
     Gameplay=&RandomRangeVector(Minimum=<0 100 20>, Maximum=<1 100 30>)
     Front End=&RandomRangeVector(Minimum=<0 -100 20>, Maximum=<1 -100 30>)
    )
   )
   &AddVelocityFromPoint(Module.Use Falloff Distance=false, Origin Offset Coordinate Space=::Simulation, Velocity Strength=50)
   &AddVelocity(
    Velocity=&FN_FrontEndVector3Modifier(Gameplay=<0 400 -120>, Front End=<0 800 -130>)
    Velocity Speed Scale=&FloatFromCurve(FloatCurve=DataInterface() {propertyValues {Curve {keys [<at 0.1 0>,<at 0.25 5>,<at 0.5 2>]}}}, CurveIndex=$System.Timeline001.Time)
   ) {module=AddVelocity001}
  ]
  particleUpdateScript [
   &SkeletalMeshLocation(
    MaskSpawnGroup=false
    Mesh Sampling Type=::NewEnumerator5
    Position Sampling=::NewEnumerator1
    Sampled Position Offset=&RandomRangeVector(Minimum=<-30 0 0>, Maximum=<30 1 1>)
    Skeletal Mesh=$User.SkeletalMesh
    Socket Sampling Mode=::NewEnumerator5
   ) {module=SkeletalMeshLocation001}
   &ParticleState()
   /Niagara/Modules/Update/Forces/CurlNoiseForce(Noise Frequency=0.5, Noise Strength=1000, Pan Noise=true)
   &Drag(Use Linear Drag=true, Use Rotational Drag=false)
   &SolveForcesAndVelocity(Clamp Velocity=false, Limit Acceleration=false)
   &ScaleColor(
    Scale Alpha=&Multiply_Float(
     A=&FloatFromCurve(
      FloatCurve=DataInterface() {propertyValues {Curve {keys [<at 0 0>,<at 0.2 1>,<at 1 1>]}}}
      Scale Curve=&ClampFloat(Float=$User.ForwardDot, Min=&FN_FrontEndFloatModifier(Gameplay=1, Front End=1))
     )
     B=1
    )
    ScaleA=true
    ScaleRGB=true
   )
   &ScaleSpriteSize(Scale Factor=&Vector2DFromFloat(Value=&FloatFromCurve(FloatCurve=DataInterface() {propertyValues {Curve {keys [<at 0 0>,<at 0.2 1.5>,<at 1 0.2>]}}}, Scale Curve=2)))
   &CameraOffset(Camera Offset Amount=5)
  ]
  rendererClasses [&NiagaraSpriteRendererProperties]
  renderers [{rendererClass=&NiagaraSpriteRendererProperties}]
 }
 SpeedLine {
  emitterUpdateScript [
   &EmitterState(
    Delay First Loop Only=true
    Inactive Response=::NewEnumerator1
    Life Cycle Mode=::NewEnumerator1
    Loop Delay=&FN_FrontEndFloatModifier(Gameplay=3.5, Front End=3.5)
    Loop Duration=3
    UseLoopDelay=true
   )
   &SpawnRate(
    SpawnRate=&FN_FrontEndFloatModifier(
     Gameplay=&Multiply_Float(
      A=&Multiply_Float(A=200, B=1)
      B=&Lerp_Float(A=-0.8, B=1, Alpha=$System.FN_Timer.Value)
     )
     Front End=60
    )
    Use Spawn Probability=false
   )
  ]
  particleSpawnScript [
   &InitializeParticle(
    Lifetime Max=0.12
    Lifetime Min=0.1
    Lifetime Mode=::NewEnumerator1
    Position Offset=&FN_FrontEndVector3Modifier(Gameplay=<0 0 0>, Front End=<0 170 0>)
    Sprite Size Max=<2 7>
    Sprite Size Min=<2 2>
    Sprite Size Mode=::NewEnumerator2
    UsePositionOffset=true
    Write Material Random=true
   )
   &TorusLocation(
    Align Torus With Axis=true
    Handle Radius=0
    Large Radius:NiagaraFloat=&RandomRangeFloat(
     Minimum=&FN_FrontEndFloatModifier(Gameplay=250, Front End=100)
     Maximum=&FN_FrontEndFloatModifier(Gameplay=250, Front End=150)
    )
    MaskSpawnGroup=false
    Module.Override Randomness=false
    Offset=&FN_FrontEndVector3Modifier(Gameplay=<0 100 0>, Front End=<0 -100 0>)
    Torus Axis=<0 1 0>
   )
   &AddVelocity(Velocity=&RandomRangeVector(Minimum=<0 500 0>, Maximum=<0 500 0>))
   &AddVelocityFromPoint(
    Module.Use Falloff Distance=false
    Origin Offset=&FN_FrontEndVector3Modifier(Gameplay=<0 -200 0>, Front End=<0 -200 0>)
    Velocity Strength=&FN_FrontEndFloatModifier(Gameplay=1500, Front End=300)
   )
   &CameraOffset(Camera Offset Amount=-10)
  ]
  particleUpdateScript [
   &SolveForcesAndVelocity(Clamp Velocity=false, Limit Acceleration=false)
   &ParticleState()
   &ScaleSpriteSizeBySpeed(Max Scale Factor=<1 50>, Min Scale Factor=<0.01 0.05>, Velocity Threshold=1)
   &ScaleColor(
    Scale Alpha=&Multiply_Float(
     A=&FloatFromCurve(
      FloatCurve=DataInterface() {propertyValues {Curve {keys [
          {interpMode=RCIM_Cubic, tangentMode=RCTM_User, time=0, value=0}
          {interpMode=RCIM_Cubic, tangentMode=RCTM_User, time=0.5, value=1}
          {tangentMode=RCTM_User, time=1, value=0}
         ]}}}
      Scale Curve=$System.TOD
     )
     B=&Multiply_Float(
      A=&RandomRangeFloat(Minimum=0.1, Maximum=0.3)
      B=&FN_FrontEndFloatModifier(Gameplay=1, Front End=0.5)
     )
    )
    ScaleA=true
    ScaleRGB=true
   )
  ]
  rendererClasses [&NiagaraSpriteRendererProperties]
  renderers [{rendererClass=&NiagaraSpriteRendererProperties}]
 }
 ThreeSwords_BabyTrail {
  emitterUpdateScript [
   &EmitterState(
    Inactive Response=::NewEnumerator1
    Life Cycle Mode=::NewEnumerator1
    Loop Behavior=::NewEnumerator1
    Loop Delay=$System.Delay
    Loop Duration Mode=::NewEnumerator1
    UseLoopDelay=true
   )
   &SpawnBurst_Instantaneous(Spawn Count=3, Use Spawn Probability=false, UseLoopCountLimit=false)
  ]
  particleSpawnScript [
   &FN_SetSpawnGroup(Number of Spawn Groups=3)
   &InitializeParticle(
    Color=&LerpLinearColors(StartColor=<rgba 0.4 0 0.0052 0.5>, EndColor=<rgba 0.2533 0 0.8 0.5>, LerpFactor=$User.Variant)
    Mesh Scale=<1.2 1 1>
    Mesh Scale Mode=::NewEnumerator4
    Position Offset=<-20 0 3>
    UsePositionOffset=true
    Write Material Random=true
   )
   &CameraOffset(Camera Offset Amount=-10)
   &InitialMeshOrientation(
    Mesh Orientation Mode=::NewEnumerator2
    Orientation Vector=<0 1 0>
    Rotation=&SelectVectorFromArray(Direct Array Index=$Particles.SpawnGroup, Vector Selection Array=DataInterface() {propertyValues {FloatData [<0 0 -0.01>, <0 0 0.01>]}})
    Use Rotation Vector=true
   )
   &InitialMeshOrientation(Mesh Orientation Mode=::NewEnumerator2, Rotation=<-0.025 0 0>, Use Rotation Vector=true) {module=InitialMeshOrientation001}
   /SetVariables(Particles.New Linear Color:LinearColor=<rgba 0.5 0 0 0.2>) {isSetParameters=true, module=SetVariables_#5}
  ]
  particleUpdateScript [
   &SkeletalMeshLocation(
    Bone / Socket Index=$Particles.SpawnGroup
    MaskSpawnGroup=false
    Mesh Sampling Type=::NewEnumerator5
    Sampled Position Offset=&Lerp_Vector(A=<0 20 0>, B=<0 20 -5>, Alpha=$User.ForwardDot)
    Skeletal Mesh=$User.SkeletalMesh
    Socket Sampling Mode=::NewEnumerator5
   ) {enabled=false}
   &SkeletalMeshLocation(
    Bone / Socket Index=&Add_Integer(A=1, B=&ReturnExecIndex())
    MaskSpawnGroup=false
    Mesh Sampling Type=::NewEnumerator5
    Sampled Position Offset=&SelectVectorFromArray(Direct Array Index=$Particles.SpawnGroup, Vector Selection Array=DataInterface() {propertyValues {FloatData [<0 0 0>, <0 0 0>]}})
    Skeletal Mesh=$User.SkeletalMesh
    Socket Sampling Mode=::NewEnumerator5
   ) {module=SkeletalMeshLocation001}
   /SetVariables(Particles.Glow Dy:Vector4f=&MakeVector4(
     X=&Multiply_Float(
      A=1
      B=&Lerp_Float(A=1, B=0.5, Alpha=&ClampFloat(Float=$System.InterpolateOverTime.InterpolateOverTime_Float.Moving Average))
     )
     Y=&SelectFloatFromArray(Direct Array Index=$Particles.SpawnGroup, Float Selection Array=DataInterface() {propertyValues {FloatData [0,0.15]}})
     Z=&SelectFloatFromArray(Direct Array Index=$Particles.SpawnGroup, Float Selection Array=DataInterface() {propertyValues {FloatData [-1,1]}})
     W=&FloatFromCurve(FloatCurve=DataInterface() {propertyValues {Curve {keys [<at 0 0>,<at 1 1>]}}})
    )) {isSetParameters=true, module=SetVariables_#6}
   /SetVariables(Particles.Glow_Color:LinearColor=&MakeLinearColorFromVectorAndFloat(
     "Vector (RGB)"=&MakeVectorFromLinearColorRGB(Color=&MultiplyLinearColor(
       Linear Color=<rgba 0 1.4351 1.7 0.5>
       Scale Factor=&MakeVector4_FromFloat(Float=&Multiply_Float(A=$System.TOD, B=&FN_FrontEndFloatModifier(Gameplay=0.9, Front End=0.6)))
      ))
     "Float (Alpha)"=&Multiply_Float(A=0.5, B=1)
    )) {isSetParameters=true, module=SetVariables_#7}
   &ParticleState(Kill Particles When Lifetime Has Elapsed=false, Let Infinitely Lived Particles Die When Emitter Deactivates=true)
   &ScaleMeshSize(Scale Factor=&MakeVector(
     X=&Lerp_Float(A=1.2, B=1.8, Alpha=&ClampFloat(Float=$System.InterpolateOverTime.InterpolateOverTime_Float.Moving Average))
     Y=&Lerp_Float(A=1, B=1, Alpha=&ClampFloat(Float=$System.InterpolateOverTime.InterpolateOverTime_Float.Moving Average))
     Z=&Lerp_Float(A=1, B=1, Alpha=&ClampFloat(Float=$System.InterpolateOverTime.InterpolateOverTime_Float.Moving Average))
    ))
   &DynamicMaterialParameters(
    Index 0 Param 1=&Multiply_Float(
     A=1
     B=&Lerp_Float(A=1, B=0.5, Alpha=&ClampFloat(Float=$System.InterpolateOverTime.InterpolateOverTime_Float.Moving Average))
    )
    Index 0 Param 2=&SelectFloatFromArray(Direct Array Index=$Particles.SpawnGroup, Float Selection Array=DataInterface() {propertyValues {FloatData [0,0.15]}})
    Index 0 Param 3=&SelectFloatFromArray(Direct Array Index=$Particles.SpawnGroup, Float Selection Array=DataInterface() {propertyValues {FloatData [-1,1]}})
    Index 0 Param 4=&Multiply_Float(
     A=&FloatFromCurve(FloatCurve=DataInterface() {propertyValues {Curve {keys [<at 0 0>,<at 1 1>]}}})
     B=&ClampFloat(Float=$System.InterpolateOverTime.InterpolateOverTime_Float.Moving Average)
    )
    Module.Param0WriteEnabled=true
    Module.Param1WriteEnabled=true
    Module.Param2WriteEnabled=true
    Module.Param3WriteEnabled=true
   )
   /SetVariables(Particles.Glow02 dy:Vector4f=&MakeVector4(
     X=&Multiply_Float(
      A=1
      B=&Lerp_Float(A=1, B=0.5, Alpha=&ClampFloat(Float=$System.InterpolateOverTime.InterpolateOverTime_Float.Moving Average))
     )
     Y=&SelectFloatFromArray(Direct Array Index=$Particles.SpawnGroup, Float Selection Array=DataInterface() {propertyValues {FloatData [0,0.15]}})
     Z=&SelectFloatFromArray(Direct Array Index=$Particles.SpawnGroup, Float Selection Array=DataInterface() {propertyValues {FloatData [-1,1]}})
     W=&Multiply_Float(
      A=&FloatFromCurve(FloatCurve=DataInterface() {propertyValues {Curve {keys [<at 0 0>,<at 1 1>]}}})
      B=&ClampFloat(Float=$System.InterpolateOverTime.InterpolateOverTime_Float.Moving Average)
     )
    )) {isSetParameters=true, module=SetVariables_#8}
   /SetVariables(Particles.Glow02 Color:LinearColor=&MakeLinearColorFromVectorAndFloat(
     "Vector (RGB)"=&MakeVectorFromLinearColorRGB(Color=&MultiplyLinearColor(
       Linear Color=&LerpLinearColors(StartColor=<rgba 1.5 0 0 1>, EndColor=<rgba 2.4 0 4 0.5>, LerpFactor=$User.Variant)
       Scale Factor=&MakeVector4_FromFloat(Float=&Multiply_Float(A=$System.TOD, B=&FN_FrontEndFloatModifier(Gameplay=1, Front End=1)))
      ))
     "Float (Alpha)"=&Multiply_Float(A=0.5, B=1)
    )) {isSetParameters=true, module=SetVariables_#9}
   /SetVariables(Particles.Outline Color:LinearColor=<rgba 1 0 0.0167 1>) {isSetParameters=true, module=SetVariables_#10}
   &ScaleColor(
    Scale RGB=&Multiply_VectorByFloat(
     Vector=&Multiply_VectorByFloat(Vector=<1 1 1>, Float=&FN_FrontEndFloatModifier(Gameplay=1.1, Front End=1))
     Float=$System.TOD
    )
    ScaleA=true
    ScaleRGB=true
   )
  ]
  rendererClasses [&NiagaraMeshRendererProperties]
  renderers [
   {rendererClass=&NiagaraMeshRendererProperties, rendererIndex=0}
   {rendererClass=&NiagaraMeshRendererProperties, rendererIndex=1}
   {rendererClass=&NiagaraMeshRendererProperties, rendererIndex=2}
   {rendererClass=&NiagaraMeshRendererProperties, rendererIndex=3}
   {rendererClass=&NiagaraMeshRendererProperties, rendererIndex=4}
  ]
 }
 ThreeSwords_IdleGlow {
  emitterUpdateScript [
   &EmitterState(
    Inactive Response=::NewEnumerator1
    Life Cycle Mode=::NewEnumerator1
    Loop Behavior=::NewEnumerator1
    Loop Duration Mode=::NewEnumerator1
    UseLoopDelay=false
   )
   &SpawnBurst_Instantaneous(Spawn Count=3, Use Spawn Probability=false, UseLoopCountLimit=false)
  ]
  particleSpawnScript [
   &InitializeParticle(
    Color=&LerpLinearColors(StartColor=<rgba 2 0 0 0.5>, EndColor=<rgba 1.3333 0 2 0.5>, LerpFactor=$User.Variant)
    Lifetime=5
    Position Offset=<0 7 50>
    Sprite Size Mode=::NewEnumerator3
    Uniform Sprite Size=80
    UsePositionOffset=true
    Write Material Random=true
   )
   &FN_SetSpawnGroup(Number of Spawn Groups=3)
   &PartitionParticles(Number Of Partitions=3)
  ]
  particleUpdateScript [
   &ParticleState(Kill Particles When Lifetime Has Elapsed=false)
   &SkeletalMeshLocation(
    Bone / Socket Index=&Add_Integer(A=1, B=$Particles.PartitionParticles.Partition)
    MaskSpawnGroup=false
    Mesh Sampling Type=::NewEnumerator5
    Sampled Position Offset=<0 -30 0>
    Skeletal Mesh=$User.SkeletalMesh
    Socket Sampling Mode=::NewEnumerator5
   )
   &ScaleColor(
    Scale Alpha=&Lerp_Float(
     A=1
     B=&Add_Float(A=1, B=$System.CurrentLoopDuration)
     Alpha=&FloatFromCurve(
      FloatCurve=DataInterface() {propertyValues {Curve {keys [
          {interpMode=RCIM_Cubic, tangentMode=RCTM_User, time=0, value=0}
          {interpMode=RCIM_Cubic, tangentMode=RCTM_User, time=0.5, value=1}
          {tangentMode=RCTM_User, time=1, value=0}
         ]}}}
      CurveIndex=$System.NormalizedLoopAge
      Scale Curve=0.33
     )
    )
    ScaleA=true
    ScaleRGB=false
   )
   &DynamicMaterialParameters(
    Index 0 Param 1=&MakeFloatFromVector(Vector=$Engine.Owner.SystemYAxis, Channel=::NewEnumerator0)
    Index 0 Param 2=&MakeFloatFromVector(Vector=$Engine.Owner.SystemYAxis, Channel=::NewEnumerator1)
    Index 0 Param 3=&MakeFloatFromVector(Vector=$Engine.Owner.SystemYAxis, Channel=::NewEnumerator2)
    Index 0 Param 4=150
    Module.Param0WriteEnabled=true
    Module.Param1WriteEnabled=true
    Module.Param2WriteEnabled=true
    Module.Param3WriteEnabled=true
   )
  ]
  rendererClasses [&NiagaraSpriteRendererProperties, &NiagaraLightRendererProperties]
  renderers [
   {rendererClass=&NiagaraSpriteRendererProperties}
   {rendererClass=&NiagaraLightRendererProperties, rendererIndex=1}
  ]
 }
 ThreeSwords_LeanGlowFlash {
  emitterUpdateScript [
   &EmitterState(
    Inactive Response=::NewEnumerator1
    Life Cycle Mode=::NewEnumerator1
    Loop Behavior=::NewEnumerator1
    Loop Duration Mode=::NewEnumerator1
    UseLoopDelay=false
   )
   &SpawnBurst_Instantaneous(Spawn Count=3, Use Spawn Probability=false, UseLoopCountLimit=false)
  ]
  particleSpawnScript [
   &InitializeParticle(
    Color=&LerpLinearColors(StartColor=<rgba 2 0 0 4>, EndColor=<rgba 1.3 0 2 4>, LerpFactor=$User.Variant)
    Lifetime=0.5
    Position Offset=<0 7 50>
    Sprite Size Mode=::NewEnumerator3
    Uniform Sprite Size=80
    UsePositionOffset=true
    Write Material Random=true
   )
   &FN_SetSpawnGroup(Number of Spawn Groups=3)
   &PartitionParticles(Number Of Partitions=3)
  ]
  particleUpdateScript [
   &ParticleState(Kill Particles When Lifetime Has Elapsed=false)
   &SkeletalMeshLocation(
    Bone / Socket Index=&Add_Integer(A=1, B=$Particles.PartitionParticles.Partition)
    MaskSpawnGroup=false
    Mesh Sampling Type=::NewEnumerator5
    Sampled Position Offset=<0 -30 0>
    Skeletal Mesh=$User.SkeletalMesh
    Socket Sampling Mode=::NewEnumerator5
   )
   &ScaleColor(
    Scale Alpha=&FloatFromCurve(
     FloatCurve=DataInterface() {propertyValues {Curve {keys [{tangentMode=RCTM_User, time=0, value=1}, {tangentMode=RCTM_User, time=1, value=0}]}}}
     Scale Curve=10
    )
    ScaleA=true
    ScaleRGB=false
   ) {enabled=false}
   &DynamicMaterialParameters(
    Index 0 Param 1=&MakeFloatFromVector(Vector=$Engine.Owner.SystemYAxis, Channel=::NewEnumerator0)
    Index 0 Param 2=&MakeFloatFromVector(Vector=$Engine.Owner.SystemYAxis, Channel=::NewEnumerator1)
    Index 0 Param 3=&MakeFloatFromVector(Vector=$Engine.Owner.SystemYAxis, Channel=::NewEnumerator2)
    Index 0 Param 4=150
    Module.Param0WriteEnabled=true
    Module.Param1WriteEnabled=true
    Module.Param2WriteEnabled=true
    Module.Param3WriteEnabled=true
   )
   &ScaleColor(
    Scale Alpha=&FloatFromCurve(
     FloatCurve=DataInterface() {propertyValues {Curve {keys [<at -0.7 0>,<at -0.4 1>,<at -0.2 0>,<at 0.2 0>,<at 0.4 1>,<at 0.7 0>]}}}
     CurveIndex=&Subtract_Float(A=$System.Timeline.Time, B=&ReturnNormalizedExecIndex(Normalized Index Scale=0.1))
    )
    ScaleA=true
    ScaleRGB=true
   ) {module=ScaleColor001}
  ]
  rendererClasses [&NiagaraSpriteRendererProperties, &NiagaraLightRendererProperties]
  renderers [
   {rendererClass=&NiagaraSpriteRendererProperties}
   {rendererClass=&NiagaraLightRendererProperties, rendererIndex=1}
  ]
 }
 Trail {
  emitterUpdateScript [
   &EmitterState(
    Inactive Response=::NewEnumerator1
    Life Cycle Mode=::NewEnumerator1
    Loop Behavior=::NewEnumerator1
    Loop Delay=$System.Delay
    Loop Duration Mode=::NewEnumerator1
    UseLoopDelay=true
   )
   &SpawnBurst_Instantaneous(Use Spawn Probability=false, UseLoopCountLimit=false)
  ]
  particleSpawnScript [
   &FN_SetSpawnGroup(Number of Spawn Groups=1)
   &InitializeParticle(
    Color=&LerpLinearColors(StartColor=<rgba 0.5 0 0.0065 0.3>, EndColor=<rgba 0.3867 0 0.8 0.2>, LerpFactor=$User.Variant)
    Mesh Scale=<1.2 1 1>
    Mesh Scale Mode=::NewEnumerator4
    Position Offset=<-20 0 3>
    UsePositionOffset=true
    Write Material Random=true
   )
   &CameraOffset()
   &InitialMeshOrientation(
    Mesh Orientation Mode=::NewEnumerator2
    Orientation Vector=<0 1 0>
    Rotation=&SelectVectorFromArray(Direct Array Index=$Particles.SpawnGroup, Vector Selection Array=DataInterface() {propertyValues {FloatData [<0 0 -0.01>, <0 0 0.01>]}})
    Use Rotation Vector=true
   )
   &InitialMeshOrientation(
    Mesh Orientation Mode=::NewEnumerator2
    Rotation=&FN_FrontEndVector3Modifier(Gameplay=<0.01 0 0>, Front End=<0 0 0>)
    Use Rotation Vector=true
   ) {module=InitialMeshOrientation001}
   /SetVariables(Particles.New Linear Color:LinearColor=<rgba 0.5 0 0 0.2>) {isSetParameters=true, module=SetVariables_#5}
  ]
  particleUpdateScript [
   &SkeletalMeshLocation(
    Bone / Socket Index=$Particles.SpawnGroup
    MaskSpawnGroup=false
    Mesh Sampling Type=::NewEnumerator5
    Sampled Position Offset=<0 20 2>
    Skeletal Mesh=$User.SkeletalMesh
    Socket Sampling Mode=::NewEnumerator5
   )
   /SetVariables(Particles.Glow Dy:Vector4f=&MakeVector4(
     X=&Multiply_Float(
      A=1
      B=&Lerp_Float(A=1, B=0.5, Alpha=&ClampFloat(Float=$System.InterpolateOverTime.InterpolateOverTime_Float.Moving Average))
     )
     Y=&SelectFloatFromArray(Direct Array Index=$Particles.SpawnGroup, Float Selection Array=DataInterface() {propertyValues {FloatData [0,0.15]}})
     Z=&SelectFloatFromArray(Direct Array Index=$Particles.SpawnGroup, Float Selection Array=DataInterface() {propertyValues {FloatData [-1,1]}})
     W=&FloatFromCurve(FloatCurve=DataInterface() {propertyValues {Curve {keys [<at 0 0>,<at 1 1>]}}})
    )) {isSetParameters=true, module=SetVariables_#6}
   /SetVariables(Particles.Glow_Color:LinearColor=&MakeLinearColorFromVectorAndFloat(
     "Vector (RGB)"=&MakeVectorFromLinearColorRGB(Color=&MultiplyLinearColor(
       Linear Color=<rgba 0 1.4351 1.7 0.5>
       Scale Factor=&MakeVector4_FromFloat(Float=&Multiply_Float(A=$System.TOD, B=&FN_FrontEndFloatModifier(Gameplay=0.9, Front End=0.6)))
      ))
     "Float (Alpha)"=&Multiply_Float(A=0.5, B=1)
    )) {isSetParameters=true, module=SetVariables_#7}
   &ParticleState(Kill Particles When Lifetime Has Elapsed=false, Let Infinitely Lived Particles Die When Emitter Deactivates=true)
   &ScaleMeshSize(Scale Factor=&FN_FrontEndVector3Modifier(
     Gameplay=&MakeVector(
      X=&Lerp_Float(A=1.2, B=3, Alpha=&ClampFloat(Float=$System.InterpolateOverTime.InterpolateOverTime_Float.Moving Average))
      Y=&Lerp_Float(A=0.7, B=1, Alpha=&ClampFloat(Float=$System.InterpolateOverTime.InterpolateOverTime_Float.Moving Average))
      Z=&Lerp_Float(A=1, B=1, Alpha=&ClampFloat(Float=$System.InterpolateOverTime.InterpolateOverTime_Float.Moving Average))
     )
     Front End=<3 1 1>
    ))
   &DynamicMaterialParameters(
    Index 0 Param 1=&Multiply_Float(
     A=1
     B=&Lerp_Float(A=1, B=0.5, Alpha=&ClampFloat(Float=$System.InterpolateOverTime.InterpolateOverTime_Float.Moving Average))
    )
    Index 0 Param 2=&SelectFloatFromArray(Direct Array Index=$Particles.SpawnGroup, Float Selection Array=DataInterface() {propertyValues {FloatData [0,0.15]}})
    Index 0 Param 3=&SelectFloatFromArray(Direct Array Index=$Particles.SpawnGroup, Float Selection Array=DataInterface() {propertyValues {FloatData [-1,1]}})
    Index 0 Param 4=&Multiply_Float(
     A=&FloatFromCurve(FloatCurve=DataInterface() {propertyValues {Curve {keys [<at 0 0>,<at 1 1>]}}})
     B=&ClampFloat(Float=$System.InterpolateOverTime.InterpolateOverTime_Float.Moving Average)
    )
    Module.Param0WriteEnabled=true
    Module.Param1WriteEnabled=true
    Module.Param2WriteEnabled=true
    Module.Param3WriteEnabled=true
   )
   /SetVariables(Particles.Glow02 dy:Vector4f=&MakeVector4(
     X=&Multiply_Float(
      A=1
      B=&Lerp_Float(A=1, B=0.5, Alpha=&ClampFloat(Float=$System.InterpolateOverTime.InterpolateOverTime_Float.Moving Average))
     )
     Y=&SelectFloatFromArray(Direct Array Index=$Particles.SpawnGroup, Float Selection Array=DataInterface() {propertyValues {FloatData [0,0.15]}})
     Z=&SelectFloatFromArray(Direct Array Index=$Particles.SpawnGroup, Float Selection Array=DataInterface() {propertyValues {FloatData [-1,1]}})
     W=&Multiply_Float(
      A=&FloatFromCurve(FloatCurve=DataInterface() {propertyValues {Curve {keys [<at 0 0>,<at 1 1>]}}})
      B=&ClampFloat(Float=$System.InterpolateOverTime.InterpolateOverTime_Float.Moving Average)
     )
    )) {isSetParameters=true, module=SetVariables_#8}
   /SetVariables(Particles.Glow02 Color:LinearColor=&MakeLinearColorFromVectorAndFloat(
     "Vector (RGB)"=&MakeVectorFromLinearColorRGB(Color=&MultiplyLinearColor(
       Linear Color=&LerpLinearColors(StartColor=<rgba 1.5 0 0 1>, EndColor=<rgba 2.4 0 3 1>, LerpFactor=$User.Variant)
       Scale Factor=&MakeVector4_FromFloat(Float=&Multiply_Float(A=$System.TOD, B=&FN_FrontEndFloatModifier(Gameplay=1, Front End=1)))
      ))
     "Float (Alpha)"=&Multiply_Float(A=0.5, B=1)
    )) {isSetParameters=true, module=SetVariables_#9}
   /SetVariables(Particles.Outline Color:LinearColor=<rgba 1 0 0.0167 1>) {enabled=false, isSetParameters=true, module=SetVariables_#10}
   &ScaleColor(
    Scale RGB=&Multiply_VectorByFloat(
     Vector=&Multiply_VectorByFloat(Vector=<1 1 1>, Float=&FN_FrontEndFloatModifier(Gameplay=1.1, Front End=1))
     Float=$System.TOD
    )
    ScaleA=true
    ScaleRGB=true
   )
  ]
  rendererClasses [&NiagaraMeshRendererProperties]
  renderers [
   {rendererClass=&NiagaraMeshRendererProperties, rendererIndex=0}
   {rendererClass=&NiagaraMeshRendererProperties, rendererIndex=1}
   {rendererClass=&NiagaraMeshRendererProperties, rendererIndex=2}
   {rendererClass=&NiagaraMeshRendererProperties, rendererIndex=3}
   {rendererClass=&NiagaraMeshRendererProperties, rendererIndex=4}
  ]
 }
 Underneath {
  emitterUpdateScript [
   &EmitterState(
    Inactive Response=::NewEnumerator1
    Life Cycle Mode=::NewEnumerator1
    Loop Behavior=::NewEnumerator1
    Loop Delay=$System.Delay
    Loop Duration Mode=::NewEnumerator1
    UseLoopDelay=true
   )
   &SpawnBurst_Instantaneous(Use Spawn Probability=false, UseLoopCountLimit=false)
  ]
  particleSpawnScript [
   &FN_SetSpawnGroup(Number of Spawn Groups=1)
   &InitializeParticle(
    Color=&LerpLinearColors(StartColor=<rgba 0.4 0 0.0055 0.335>, EndColor=<rgba 0.35 0 1 0.2>, LerpFactor=$User.Variant)
    Mesh Scale=&Multiply_VectorByFloat(Vector=<3 1 1>, Float=0.8)
    Mesh Scale Mode=::NewEnumerator4
    Position Offset=<-40 1.5 3>
    UsePositionOffset=true
    Write Material Random=true
   )
   &DynamicMaterialParameters(
    Index 0 Param 1=2
    Index 0 Param 2=&SelectFloatFromArray(Direct Array Index=$Particles.SpawnGroup, Float Selection Array=DataInterface() {propertyValues {FloatData [0,0.5,0.25,0.75]}})
    Index 0 Param 3=&SelectFloatFromArray(Direct Array Index=$Particles.SpawnGroup, Float Selection Array=DataInterface() {propertyValues {FloatData [-1,1,-1,1]}})
    Index 0 Param 4=1
    Module.Param0WriteEnabled=true
    Module.Param1WriteEnabled=true
    Module.Param2WriteEnabled=true
    Module.Param3WriteEnabled=true
   )
   &CameraOffset()
   &InitialMeshOrientation(Mesh Orientation Mode=::NewEnumerator2, Orientation Vector=<0 1 0>, Use Rotation Vector=true)
  ]
  particleUpdateScript [
   &SkeletalMeshLocation(
    MaskSpawnGroup=false
    Mesh Sampling Type=::NewEnumerator5
    Sampled Position Offset=&FN_FrontEndVector3Modifier(
     Gameplay=&Lerp_Vector(A=<0 -12 8>, B=<0 -20 8>, Alpha=$System.InterpolateOverTime.InterpolateOverTime_Float.Moving Average)
     Front End=<0 -30 9>
    )
    Skeletal Mesh=$User.SkeletalMesh
    Socket Sampling Mode=::NewEnumerator5
   ) {module=SkeletalMeshLocation001}
   &DynamicMaterialParameters(
    Index 0 Param 1=&Multiply_Float(
     A=1
     B=&Lerp_Float(A=1, B=0.5, Alpha=&ClampFloat(Float=$System.InterpolateOverTime.InterpolateOverTime_Float.Moving Average))
    )
    Index 0 Param 2=&SelectFloatFromArray(Direct Array Index=$Particles.SpawnGroup, Float Selection Array=DataInterface() {propertyValues {FloatData [0,0.15]}})
    Index 0 Param 3=&SelectFloatFromArray(Direct Array Index=$Particles.SpawnGroup, Float Selection Array=DataInterface() {propertyValues {FloatData [-1,1]}})
    Index 0 Param 4=&Multiply_Float(
     A=&FloatFromCurve(FloatCurve=DataInterface() {propertyValues {Curve {keys [<at 0 0>,<at 1 1>]}}})
     B=&ClampFloat(Float=$System.InterpolateOverTime.InterpolateOverTime_Float.Moving Average)
    )
    Module.Param0WriteEnabled=false
    Module.Param1WriteEnabled=false
    Module.Param2WriteEnabled=false
    Module.Param3WriteEnabled=true
   ) {module=DynamicMaterialParameters001}
   &ParticleState(Kill Particles When Lifetime Has Elapsed=false, Let Infinitely Lived Particles Die When Emitter Deactivates=true)
   &ScaleMeshSize(Scale Factor=&Multiply_Vector(
     A=<0.6 1.5 1>
     B=&FN_FrontEndVector3Modifier(
      Gameplay=&Multiply_Vector(
       A=&MakeVector(
        X=&FloatFromCurve(FloatCurve=DataInterface() {propertyValues {Curve {keys [<at 0.2 1>,<at 0.25 1.3>,<at 0.4 1>]}}}, CurveIndex=$System.Timeline001.Time)
        Y=1
        Z=1
       )
       B=&MakeVector(
        X=&Lerp_Float(A=0.6, B=1, Alpha=&ClampFloat(Float=$System.InterpolateOverTime.InterpolateOverTime_Float.Moving Average))
        Y=&Lerp_Float(A=1, B=1, Alpha=&ClampFloat(Float=$System.InterpolateOverTime.InterpolateOverTime_Float.Moving Average))
        Z=&Lerp_Float(A=2, B=1, Alpha=&ClampFloat(Float=$System.InterpolateOverTime.InterpolateOverTime_Float.Moving Average))
       )
      )
      Front End=<1 1 2>
     )
    ))
   /SetVariables(Particles.Outline color:LinearColor=&LerpLinearColors(StartColor=<rgba 1 0 0.0167 0.5>, EndColor=<rgba 1.7 0 3 0.45>, LerpFactor=$User.Variant)) {isSetParameters=true}
   &ScaleColor(
    Scale RGB=&Multiply_VectorByFloat(
     Vector=<1 1 1>
     Float=&Multiply_Float(A=$System.TOD, B=&FN_FrontEndFloatModifier(Gameplay=1, Front End=1))
    )
    ScaleA=true
    ScaleRGB=true
   )
   /SetVariables() {isSetParameters=true}
  ]
  rendererClasses [&NiagaraMeshRendererProperties]
  renderers [
   {rendererClass=&NiagaraMeshRendererProperties, rendererIndex=0}
   {rendererClass=&NiagaraMeshRendererProperties, rendererIndex=1}
  ]
 }
}
summary {
 systemName=NS_Glider_Ebony_V3
 emitters [
  {emitterName=Sparks, rendererClasses [&NiagaraSpriteRendererProperties]}
  {emitterName=Underneath, rendererClasses [&NiagaraMeshRendererProperties]}
  {emitterName=SpeedLine, rendererClasses [&NiagaraSpriteRendererProperties]}
  {emitterName=Trail, rendererClasses [&NiagaraMeshRendererProperties]}
  {emitterName=Flare, rendererClasses [&NiagaraSpriteRendererProperties, &NiagaraLightRendererProperties]}
  {
   emitterName=GLOW_REACTORS001
   rendererClasses [&NiagaraSpriteRendererProperties, &NiagaraLightRendererProperties]
  }
  {emitterName=ElectricArcsSmall, rendererClasses [&NiagaraSpriteRendererProperties]}
  {
   emitterName=ThreeSwords_IdleGlow
   rendererClasses [&NiagaraSpriteRendererProperties, &NiagaraLightRendererProperties]
  }
  {bEnabled=false, emitterName=FN_Debug, rendererClasses [&NiagaraSpriteRendererProperties]}
  {
   emitterName=ThreeSwords_LeanGlowFlash
   rendererClasses [&NiagaraSpriteRendererProperties, &NiagaraLightRendererProperties]
  }
  {emitterName=ThreeSwords_BabyTrail, rendererClasses [&NiagaraMeshRendererProperties]}
  {emitterName=FakeLight, rendererClasses [&NiagaraSpriteRendererProperties]}
 ]
 userVariables [
  User.bIsFrontEnd {type=NiagaraBool}
  User.bIsFrontEndPreview {type=NiagaraBool}
  User.bIsFullyDeployed {type=NiagaraBool}
  User.Debug {type=NiagaraFloat, defaultValue=0, description=""}
  User.Delay {type=NiagaraFloat, defaultValue=0, description=""}
  User.EngineDelay {type=NiagaraFloat, defaultValue=0, description=""}
  User.ForwardDot {type=NiagaraFloat, defaultValue=0, description=""}
  User.PlayerSpeed {type=NiagaraFloat, defaultValue=0, description=""}
  User.RightDot {type=NiagaraFloat, defaultValue=0, description=""}
  User.SkeletalMesh {
   type=NiagaraDataInterfaceSkeletalMesh
   defaultValue {
    struct=NiagaraExt_VariableValue_DataInterface
    value {dataInterface=&self:NiagaraDataInterfaceSkeletalMesh_0}
   }
  }
  User.Variant {type=NiagaraFloat, defaultValue=0, description=""}
 ]
}
